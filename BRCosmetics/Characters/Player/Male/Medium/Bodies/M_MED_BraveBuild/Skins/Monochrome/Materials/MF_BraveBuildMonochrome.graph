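# Material function for the Monochrome BraveBuild player skin that remaps base color between two tint colors, keeping red-dominant areas, and adds animated static-like emission.
# Noise, scanline and strobe textures project onto a camera-facing virtual plane near the actor, scaled by FOV and jumped in stepped time offsets.
# Base-color green blends two strobe layers, one weighted by view and up-facing normals.
# Emission is tinted by time of day and masked on downward faces.
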MaterialFunction /BRCosmetics/Characters/Player/Male/Medium/Bodies/M_MED_BraveBuild/Skins/Monochrome/Materials/MF_BraveBuildMonochrome
graph {nodes [
  n1 = GetMaterialAttributes(None=FunctionInput() {inputName=Material, inputType=FunctionInput_MaterialAttributes}) {attributeGetTypes [#1, #2]}
  n2 = LinearInterpolate(
   A=VectorParameter() {group=BraveBuild, parameterName=BB_BaseColor1, defaultValue=<rgba 0 0 0 1>}.RGB
   B=VectorParameter() {group=BraveBuild, parameterName=BB_BaseColor2, defaultValue=<rgba 1.2 1.1637 1.0812 1>}.RGB
   Alpha=@n1.BaseColor
  )
  n3 = Subtract(A=ComponentMask(None=@n1.BaseColor) {r=true}, B=ComponentMask(None=@n1.BaseColor) {g=true})
  n4 = Round(None=Saturate(None=Multiply(A=@n3) {constB=5}))
  n5 = PixelNormalWS()
  n6 = ComponentMask(None=@n5) {b=true}
  n7 = Multiply(A=@n6) {constB=-1}
  n8 = OneMinus(None=Multiply(A=Saturate(None=@n7), B=ScalarParameter() {defaultValue=1, group=BraveBuild, parameterName=BB_EmissionMask}))
  n9 = Add(A=@n6) {constB=0.4}
  n10 = Saturate(None=DotProduct(A=@n5, B=CameraVectorWS()))
  n11 = Multiply(A=Saturate(None=Multiply(A=@n9)), B=Multiply(A=@n10, B=@n10))
  n12 = ScalarParameter() {defaultValue=0.01, group=BraveBuild, parameterName=BB_Pulse-Scale}
  n13 = Constant() {desc=Needs at least a small amount of this or it will break, r=0.25}
  n14 = MaterialFunctionCall() {materialFunction=&CameraDirectionVector}
  n15 = Normalize(VectorInput=Multiply(A=Constant3Vector() {constant=<rgba 1 1 0 0>}, B=@n14.Result))
  n16 = Constant() {r=25}
  n17 = ActorPositionWS()
  n18 = Constant3Vector() {constant=<rgba 0 0 40 1>}
  n19 = MaterialFunctionCall(
   Input=MaterialFunctionCall() {materialFunction=/Engine/Functions/Engine_MaterialFunctions02/Utility/FOV}."FOV "
   Input Low=Constant() {r=0.414}
   Input High=Constant() {r=1.091}
   Target Low=Constant()
   Target High=Constant() {r=1}
  ) {materialFunction=&RemapValueRange}
  n20 = MaterialFunctionCall(
   Plane Normal Axis=Add(A=Multiply(A=@n13, B=@n15), B=@n14.Result)
   Plane Center=Add(A=Multiply(A=@n14.Result, B=@n16), B=Add(A=@n17, B=@n18))
   UV Scale=LinearInterpolate(
    A=ScalarParameter() {defaultValue=1, group=BraveBuild, parameterName=BB_Scale-LowFOV}
    B=ScalarParameter() {defaultValue=1.2, group=BraveBuild, parameterName=BB_Scale-HighFOV}
    Alpha=Saturate(None=@n19.Result)
   )
  ) {materialFunction=&VirtualPlaneCoordinates}
  n21 = ScalarParameter() {defaultValue=16, group=BraveBuild, parameterName=BB_Noise-NumberOffsets}
  n22 = Time() {bOverride_Period=true, period=1}
  n23 = Multiply(A=ScalarParameter() {defaultValue=0.9, group=BraveBuild, parameterName=BB_Noise-OffsetUV}, B=Floor(None=Multiply(A=@n21, B=@n22)))
  n24 = TextureSample(UVs=Add(A=Multiply(A=@n12, B=@n20.UVs), B=@n23)) {samplerSource=SSM_Wrap_WorldGroupSettings, texture=/Game/Global/Textures/Default/Masks_and_Gradients/T_Vert_Gradient_Sections}
  n25 = Multiply(A=@n24.R, B=ScalarParameter() {defaultValue=0.25, group=BraveBuild, parameterName=BB_StrobeB})
  n26 = ScalarParameter() {defaultValue=0.013, group=BraveBuild, parameterName=BB_Noise-Scale}
  n27 = TextureSample(UVs=Add(A=@n23, B=Multiply(A=@n26, B=@n20.UVs))) {
   samplerSource=SSM_Wrap_WorldGroupSettings
   samplerType=SAMPLERTYPE_LinearColor
   texture=/Game/Global/Textures/Default/Noise/SparseGaussianNoise256
  }
  n28 = ScalarParameter() {defaultValue=0.5, group=BraveBuild, parameterName=BB_ScanlineAmt}
  n29 = TextureSample(UVs=@n20.UVs) {samplerSource=SSM_Wrap_WorldGroupSettings, texture=/Game/Global/Textures/Default/Masks_and_Gradients/gradient_stasis}
  n30 = Add(
   A=Multiply(A=OneMinus(None=@n27.R), B=ScalarParameter() {defaultValue=0.4, group=BraveBuild, parameterName=BB_NoiseAmt})
   B=Multiply(A=Multiply(A=@n27.G, B=@n27.G), B=Multiply(A=@n28, B=@n29.R))
  )
  n31 = Multiply(A=ComponentMask(None=@n1.BaseColor) {g=true}) {constB=2}
  n32 = LinearInterpolate(
   A=Add(A=Multiply(A=@n11, B=@n25), B=@n30)
   B=Multiply(A=ScalarParameter() {defaultValue=0.6, group=BraveBuild, parameterName=BB_StrobeA}, B=Add(A=@n24.R, B=@n30))
   Alpha=SmoothStep(Value=Saturate(None=@n31)) {constMax=0.6, constMin=0.3}
  )
  n33 = VectorParameter() {group=BraveBuild, parameterName=BB_Fx-Color, defaultValue=<rgba 0.9976 1 0.9583 1>}
  n34 = Multiply(
   A=Multiply(A=Multiply(A=@n32, B=@n33.RGB), B=ScalarParameter() {defaultValue=0.05, group=BraveBuild, parameterName=BB_EmissionAdd}) {constB=0.01}
   B=MaterialFunctionCall(
    Value_Morning=Constant3Vector() {constant=<rgba 0.9375 0.8738 0.8105 1>}
    Value_Day=Constant() {r=1}
    Value_Evening=Constant3Vector() {constant=<rgba 0.8542 0.6232 0.6834 1>}
    Value_Night=Constant3Vector() {constant=<rgba 0.4915 0.6217 0.9531 1>}
   ) {materialFunction=&Scalar_Time_of_Day}.Result
  )
  FunctionOutput(None=SetMaterialAttributes!(
    MaterialAttributes=@n1.MaterialAttributes
    BaseColor=LinearInterpolate(A=@n2, B=@n1.BaseColor, Alpha=@n4)
    EmissiveColor=Multiply(A=@n8, B=@n34)
   ) {attributeSetTypes [#1, #3]})
 ]}
